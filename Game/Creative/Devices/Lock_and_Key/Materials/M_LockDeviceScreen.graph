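# Default Lit material for the Lock and Key device screen, showing a two-color icon that switches between two parameterized states.
# The Locked - Unlocked scalar and vertical UV sample a transition texture, whose red channel blends the states and green channel shifts icon UVs horizontally.
# Scrolling horizontal scanlines and a sine-based gradient add to each icon mask to mix the background and icon colors.
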Material /Game/Creative/Devices/Lock_and_Key/Materials/M_LockDeviceScreen
graph {
 properties {pixelDepthOffsetMode=PDOM_Legacy, usedShadingModels=Default Lit}
 nodes [
  n1 = VectorParameter() {parameterName=Background Color, defaultValue=<rgba 0.9583 0.0291 0.0672 1>}
  n2 = VectorParameter() {parameterName=Icon Color, defaultValue=<rgba 0.9688 0.5786 0.5206 1>}
  n3 = TextureCoordinate() {vTiling=20}
  n4 = ConstantBiasScale(None=Frac(None=Panner(Coordinate=@n3) {speedY=1})) {bias=-0.5, scale=2}
  n5 = Multiply(A=@n4, B=@n4)
  n6 = Multiply(A=TextureCoordinate()) {constB=0.5}
  n7 = ComponentMask(None=@n6) {r=true}
  n8 = ComponentMask(None=@n6) {g=true}
  n9 = OneMinus(None=Multiply(A=Sine(None=@n7), B=Sine(None=@n8)))
  n10 = Add(A=Multiply(A=ComponentMask(None=@n5) {g=true}) {constB=0.1}, B=Multiply(A=Power(Base=@n9) {constExponent=8}) {constB=0.4})
  n11 = TextureCoordinate()
  n12 = TextureSample(UVs=AppendVector(A=ScalarParameter() {group=Logic, parameterName=Locked - Unlocked, sliderMax=1}, B=ComponentMask(None=@n11) {g=true})) {samplerType=SAMPLERTYPE_LinearColor, texture=/Game/Creative/Devices/Lock_and_Key/Textures/T_LockTransition_M}
  n13 = Add(A=TextureCoordinate(), B=AppendVector(A=Multiply(A=@n12.G) {constB=0.1}, B=Constant()))
  n14 = TextureSampleParameter2D(UVs=@n13) {
   parameterName=State A Icon
   samplerType=SAMPLERTYPE_LinearColor
   texture=/Game/Creative/Devices/Lock_and_Key/Textures/Icons/T_LockIcons_Exclamation_M
  }
  n15 = TextureSampleParameter2D(UVs=@n13) {
   parameterName=State B Icon
   samplerType=SAMPLERTYPE_LinearColor
   texture=&T_LockIcons_Exclamation_M
  }
  LinearInterpolate(
   A=LinearInterpolate(A=@n1.RGB, B=@n2.RGB, Alpha=Add(A=@n10, B=@n14.R))
   B=LinearInterpolate(A=@n1.RGB, B=@n2.RGB, Alpha=Add(A=@n10, B=@n15.R))
   Alpha=@n12.R
  )
 ]
}
