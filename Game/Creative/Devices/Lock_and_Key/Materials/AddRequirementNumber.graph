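# Material function for the Lock and Key device that outputs a mask of a requirement number with an extra "X" slot.
# Threshold tests on the clamped number count digit slots, dividing a cropped rectangle sized from a bottom-right corner.
# Per-pixel power-of-ten math selects each digit from a number sheet; the requirements texture fills the leftmost slot.
# A show/hide input scales the result; a static bool switches the display limit.
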MaterialFunction /Game/Creative/Devices/Lock_and_Key/Materials/AddRequirementNumber
graph {nodes [
  n1 = FunctionInput(Preview=Constant2Vector() {g=0.95, r=0.95}) {bUsePreviewValueAsDefault=true, inputName=Digit bottom right corner, inputType=FunctionInput_Vector2, sortPriority=2}
  n2 = Multiply(
   A=NamedRerouteUsage()
   B=FunctionInput(Preview=Constant() {r=0.08}) {bUsePreviewValueAsDefault=true, inputName=Digit Width, inputType=FunctionInput_Scalar, sortPriority=3}
  )
  n3 = FunctionInput(Preview=Constant() {r=0.1}) {bUsePreviewValueAsDefault=true, inputName=Digit Height, inputType=FunctionInput_Scalar, sortPriority=4}
  n4 = MaterialFunctionCall!(UVs=TextureCoordinate(), TextureIn, UpperLeftCorner=Subtract(A=@n1, B=AppendVector(A=@n2, B=@n3)), LowerRightCorner=@n1, ExportFloat4) {materialFunction=/Engine/Functions/Engine_MaterialFunctions01/Texturing/TextureCropping}
  n5 = Multiply(A=@n4.Crop UVs, B=Constant2Vector() {g=0.1, r=1}) {constA=2, constB=0.1}
  n6 = Multiply(A=OneMinus(None=ComponentMask(None=@n5) {r=true}), B=NamedRerouteUsage()) {constB=10}
  n7 = Ceil(None=Max(A=@n6) {constB=0})
  n8 = Divide(A=NamedRerouteUsage(), B=Power(Base=Constant() {r=10}, Exp=Add(A=@n7)) {desc="10-100,000,000"})
  n9 = Floor(None=Multiply(A=Frac(None=@n8)) {constB=100})
  n10 = AppendVector(A=Frac(None=@n6), B=Add(A=ComponentMask(None=@n5) {g=true}, B=Divide(A=@n9) {constB=10}))
  n11 = Saturate(None=Multiply(
    A=TextureSample(UVs=@n10) {samplerType=SAMPLERTYPE_Masks, texture=/Game/Athena/Items/Gameplay/VendingMachine/FX/T_NumberSheet}.R
    B=@n4.Crop Mask
   ) {desc=crop mask})
  n12 = Saturate(None=Multiply(A=NamedRerouteUsage(), B=ComponentMask(None=@n4.Crop UVs) {r=true}))
  n13 = TextureSample(UVs=AppendVector(A=@n12, B=ComponentMask(None=@n4.Crop UVs) {g=true})) {samplerType=SAMPLERTYPE_LinearColor, texture=/Game/Creative/Devices/Lock_and_Key/Textures/T_LockKeyRequirements_M}
  n14 = Abs(None=Subtract(A=@n12) {constB=0.5})
  n15 = Ceil(None=OneMinus(None=Multiply(A=@n14) {constB=2}))
  n16 = Constant() {r=1}
  FunctionOutput(None=Multiply(
    A=LinearInterpolate(A=@n11, B=@n13.G, Alpha=@n15)
    B=FunctionInput(Preview=@n16) {bUsePreviewValueAsDefault=true, inputName=Show // Hide number, inputType=FunctionInput_Scalar}
   )) {outputName=Number Mask}
  n17 = StaticBool() {desc="Clamp to 999,999"}
  n18 = Clamp(
   None=FunctionInput(Preview=Constant() {r=9999999}) {bUsePreviewValueAsDefault=true, inputName=Requirement number, inputType=FunctionInput_Scalar, sortPriority=1}
   Max=StaticSwitch(True=Constant() {desc=Maximum number to display, r=999999}, False=Constant() {desc=Maximum number to display, r=8888888}, Value=@n17)
  ) {clampMode=CMODE_ClampMax, maxDefault=999999}
  Multiply(A=MaterialFunctionCall(Number=NamedRerouteDeclaration(None=@n18) {name=IncomingNumber, nodeColor=<rgba 1 0 1 1>}) {materialFunction=&DebugScalarValues}.Result)
  n19 = NamedRerouteUsage()
  n20 = Constant() {r=10}
  n21 = Constant() {r=1}
  n22 = Constant()
  n23 = NamedRerouteUsage()
  n24 = Constant()
  n25 = Constant() {r=1}
  n26 = Constant()
  n27 = Constant() {r=1}
  n28 = Add(
   A=Add(A=If(A=@n23, B=@n24, "A > B"=@n25, "A == B"=@n25, "A < B"=@n26), B=If(A=@n19, B=@n20, "A > B"=@n21, "A == B"=@n21, "A < B"=@n22))
   B=If(A=NamedRerouteUsage(), B=Constant() {r=100}, "A > B"=@n27, "A == B"=@n27, "A < B"=Constant())
  )
  n29 = Constant() {r=1}
  n30 = If(A=NamedRerouteUsage(), B=Constant() {r=1000}, "A > B"=@n29, "A == B"=@n29, "A < B"=Constant())
  n31 = NamedRerouteUsage()
  n32 = Constant() {r=10000}
  n33 = Constant() {r=1}
  n34 = Constant()
  n35 = Constant() {r=1}
  n36 = Add(
   A=Add(A=Add(A=@n28, B=@n30), B=If(A=@n31, B=@n32, "A > B"=@n33, "A == B"=@n33, "A < B"=@n34))
   B=If(A=NamedRerouteUsage(), B=Constant() {r=100000}, "A > B"=@n35, "A == B"=@n35, "A < B"=Constant())
  )
  n37 = Constant() {r=1}
  n38 = Add(A=@n36, B=If(A=NamedRerouteUsage(), B=Constant() {r=100000}, "A > B"=@n37, "A == B"=@n37, "A < B"=Constant()))
  NamedRerouteDeclaration(None=Add(A=StaticSwitch(True=@n36, False=@n38, Value=@n17), B=Constant() {desc="Add a space for the \"X\"", r=1})) {name=TotalNumberOfDigits, nodeColor=<rgba 1 0 1 1>}
  n39 = NamedRerouteUsage()
  n40 = Divide(A=@n39) {constB=100}
  n41 = Saturate(None=Floor(None=Divide(A=@n39) {constB=10}))
  n42 = Add(A=Saturate(None=Floor(None=@n40)), B=@n41)
  Reroute(None=Add(A=Multiply(A=@n42, B=@n41)) {constB=2})
 ]}
