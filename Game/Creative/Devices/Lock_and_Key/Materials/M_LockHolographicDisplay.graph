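# Translucent, two-sided unlit material for the holographic display of the Creative Lock and Key device, showing state icons, a key and a required-key count.
# Time-driven sine and sphere masks distort and glitch the frame UVs.
# Blueprint-set timestamps drive key scrolling and interaction changes, using a transition texture to blend and offset layers.
# Fresnel and spawn time fade opacity, while UV-sign math builds axis-aligned offsets to size the quad.
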Material /Game/Creative/Devices/Lock_and_Key/Materials/M_LockHolographicDisplay
graph {
 properties {blendMode=BLEND_Translucent, pixelDepthOffsetMode=PDOM_Legacy, shadingModel=MSM_Unlit, twoSided=true, usedShadingModels=Unlit}
 nodes [
  n1 = VectorParameter() {group=Colors, parameterName=Edge Color, sortPriority=12, defaultValue=<rgba 0 0.527 1 4>}
  n2 = Multiply(A=ScalarParameter() {defaultValue=1, parameterName=EdgeMultiplier}, B=Multiply(A=@n1.RGB, B=@n1.RGB))
  n3 = ComponentMask(None=TextureCoordinate()) {g=true}
  n4 = Time() {bOverride_Period=true, period=100}
  n5 = Cosine(None=Add(A=@n3, B=Multiply(A=@n4) {constB=0.1}))
  n6 = Sine(None=Add(A=@n3, B=Multiply(A=@n4) {constB=-0.3}))
  n7 = SphereMask(A=Add(A=Sine(None=@n5), B=@n6), B=Constant()) {attenuationRadius=2, hardnessPercent=0}
  n8 = Subtract(A=@n7) {constB=0.5}
  n9 = Multiply(A=@n7, B=ScalarParameter() {defaultValue=0.01, group=Animation, parameterName=UV distortion, sortPriority=20}) {constB=0.02}
  n10 = SphereMask(A=Constant() {r=1}, B=Multiply(A=Saturate(None=@n5), B=Saturate(None=@n6))) {attenuationRadius=0.05, hardnessPercent=0}
  n11 = Multiply(A=@n10, B=ScalarParameter() {defaultValue=0.05, group=Animation, parameterName=UV Glitches, sortPriority=20}) {constB=0.05}
  n12 = Add(
   A=AppendVector(A=Add(A=@n9, B=@n11), B=Constant())
   B=MaterialFunctionCall(UVs=TextureCoordinate(), Texture Scale=Constant() {r=0.8}) {materialFunction=&ScaleUVsByCenter}.UVs
  )
  n13 = TextureSample(UVs=Saturate(None=@n12)) {texture=/Game/Creative/Devices/Lock_and_Key/Textures/T_LockHologramDisplayFrame_M}
  n14 = Add(A=Multiply(A=Multiply(A=@n8) {constB=0.2}, B=@n13.G), B=@n13.G)
  n15 = Multiply(A=@n14, B=Add(A=@n13.B, B=Multiply(A=@n10) {constB=0.3}))
  n16 = VectorParameter() {group=Colors, parameterName=Icon Color, sortPriority=11, defaultValue=<rgba 0 0.527 1 10>}
  n17 = Multiply(A=@n16.RGB, B=@n16.RGB)
  n18 = ScalarParameter() {group=BP Logic, parameterName=UseKeyAnimationTime, sliderMax=1, sortPriority=11}
  n19 = Multiply(A=Subtract(A=Time(), B=@n18), B=ScalarParameter() {defaultValue=0.3, group=Animation, parameterName=Flip Flop Speed, sortPriority=20})
  n20 = Multiply(
   A=Cosine(None=Add(A=@n19) {constB=0.25})
   B=ScalarParameter() {defaultValue=0.75, group=Animation, parameterName=Transition Speed, sortPriority=20}
  )
  n21 = MaterialFunctionCall(
   TimestampA=@n18
   TransitionLength_In=ScalarParameter() {defaultValue=0.3, group=BP Logic, parameterName=UseKeyAnimationLength, sortPriority=11}
  ) {materialFunction=&MF_Timestamp_InOut_01}
  n22 = OneMinus(None=@n21.Linear)
  n23 = Saturate(None=Subtract(A=Add(A=@n20) {constB=0.5}, B=Ceil(None=@n22)))
  n24 = ComponentMask(None=TextureCoordinate()) {g=true}
  n25 = TextureSample(UVs=If(A=ScalarParameter() {parameterName=UseTransition}, "A > B"=AppendVector(A=@n23, B=@n24), "A == B"=Constant(), "A < B"=Constant() {r=1})) {samplerType=SAMPLERTYPE_LinearColor, texture=/Game/Creative/Devices/Lock_and_Key/Textures/T_LockTransition_M}
  n26 = MaterialFunctionCall!(
   Show // Hide number=Multiply(
    A=ScalarParameter() {defaultValue=1, group=BP Logic, parameterName=Show Number of Keys, sliderMax=1, sortPriority=50}
    B=OneMinus(None=@n25.R)
   )
   Requirement number=ScalarParameter() {group=BP Logic, parameterName=Number of Keys Required, sortPriority=51}
   Digit bottom right corner=Constant() {r=0.85}
   Digit Width
   Digit Height
  ) {materialFunction=&/AddRequirementNumber}
  n27 = Multiply(A=Multiply(A=@n17) {constB=0.2}, B=@n26.Number Mask)
  n28 = Frac(None=Multiply(A=ScalarParameter() {defaultValue=1, group=BP Logic, parameterName=UseKeyAnimationLoops, sortPriority=11}, B=@n21.Linear))
  n29 = Multiply(A=AppendVector(A=Constant(), B=MaterialFunctionCall!(In=@n28) {materialFunction=/Game/Effects/Fort_Effects/Materials/_Functions/MF_Smoothstep_01}.Result)) {constB=-1}
  n30 = Add(A=AppendVector(A=Multiply(A=@n25.G) {constB=0.1}, B=Constant()), B=@n12)
  n31 = MaterialFunctionCall(UVs=@n30, Texture Scale=Constant() {r=0.8}) {desc=Further scaling for key icon, materialFunction=&ScaleUVsByCenter}
  n32 = Saturate(None=@n31.UVs)
  n33 = TextureSampleParameter2D(UVs=Frac(None=Add(A=@n29, B=@n32))) {
   group=BP Logic
   parameterName=Key Texture
   samplerType=SAMPLERTYPE_LinearColor
   sortPriority=60
   texture=/Game/Creative/Devices/Lock_and_Key/Textures/T_LockHiddenKey
  }
  n34 = VectorParameter() {group=Colors, parameterName=Key color multiplier, defaultValue=<rgba 1 1 1 1>}
  n35 = Multiply(A=@n34.RGB, B=@n34.RGB)
  n36 = ComponentMask(None=@n32) {g=true}
  n37 = Multiply(A=@n31.0-1 mask, B=Sine(None=Multiply(A=@n36) {constB=0.5}))
  n38 = Multiply(A=@n33.A, B=Saturate(None=Multiply(A=@n37) {constB=3}))
  n39 = TextureSampleParameter2D(UVs=@n31.UVs) {
   parameterName=State B Icon
   samplerType=SAMPLERTYPE_LinearColor
   texture=/Game/Creative/Devices/Lock_and_Key/Textures/Icons/T_LockIcons_Exclamation_M
  }
  n40 = Add(A=Multiply(A=Multiply(A=@n33.RGB, B=@n35), B=@n38), B=Multiply(A=@n17, B=@n39.G)) {constB=0}
  n41 = Cosine(None=Time())
  n42 = ScalarParameter() {defaultValue=1, parameterName=Icon Rotation Speed}
  n43 = TextureSampleParameter2D(UVs=Rotator(Coordinate=Saturate(None=@n30), Time=Multiply(A=@n41, B=@n42))) {
   parameterName=State A Icon
   samplerType=SAMPLERTYPE_LinearColor
   texture=&T_LockIcons_Exclamation_M
  }
  n44 = ScalarParameter() {group=BP Logic, parameterName=Interaction Animation Time, sliderMax=1, sortPriority=15}
  n45 = Add(A=@n44, B=Constant() {r=1})
  n46 = Add(A=@n44, B=ScalarParameter() {defaultValue=1, group=BP Logic, parameterName=User Selected Reset Delay, sliderMax=1, sortPriority=16})
  n47 = ScalarParameter() {defaultValue=1, parameterName=Successful Unlock, sliderMax=1, sortPriority=33}
  n48 = MaterialFunctionCall(
   TimestampA=@n44
   TransitionLength_In=Constant() {r=0.5}
   TimestampB=LinearInterpolate(A=@n45, B=LinearInterpolate(A=@n45, B=@n46, Alpha=@n47), Alpha=ScalarParameter() {parameterName=Interaction Type})
   TransitionLength_Out=LinearInterpolate(A=Constant() {r=0.5}, B=Constant() {r=500000}, Alpha=ScalarParameter() {parameterName=ConstantInteract})
   Transition In Only?=StaticBool()
  ) {materialFunction=&MF_Timestamp_InOut_01}
  n49 = TextureCoordinate()
  n50 = TextureSample(UVs=AppendVector(A=@n48.Linear, B=ComponentMask(None=@n49) {g=true})) {samplerType=SAMPLERTYPE_LinearColor, texture=&T_LockTransition_M}
  n51 = LinearInterpolate(A=LinearInterpolate(A=Add(A=@n27, B=@n40), B=Multiply(A=@n17, B=@n43.R), Alpha=@n25.R), B=@n17, Alpha=@n50.R) {constB=10}
  n52 = Max(A=@n38, B=@n39.G)
  n53 = AppendVector(A=Multiply(A=@n50.G) {constB=0.1}, B=Constant())
  n54 = TextureSample(UVs=MaterialFunctionCall(UVs=Add(A=@n12, B=@n53), Texture Scale=Constant() {r=0.65}) {materialFunction=&ScaleUVsByCenter}.UVs) {samplerType=SAMPLERTYPE_LinearColor, texture=/Game/Creative/Devices/Lock_and_Key/Textures/T_LockKeyRequirements_M}
  n55 = ScalarParameter() {parameterName=Interaction Type}
  n56 = LinearInterpolate(
   A=LinearInterpolate(A=Max(A=@n52, B=@n26.Number Mask), B=@n43.R, Alpha=@n25.R)
   B=Multiply(A=LinearInterpolate(A=@n54.R, B=@n54.R, Alpha=@n55), B=Ceil(None=@n48.Linear))
   Alpha=@n50.R
  )
  n57 = VectorParameter() {group=Colors, parameterName=Background Color, sortPriority=10, defaultValue=<rgba 0.0366 0.224 0.6 1>}
  n58 = Multiply(A=@n57.RGB, B=@n14)
  Clamp(None=Add(A=LinearInterpolate(A=@n58, B=@n51, Alpha=@n56), B=Multiply(A=@n2, B=@n15)), Max=Constant() {r=3}) {clampMode=CMODE_ClampMax}
  n59 = Add(A=@n56, B=Multiply(A=@n13.G, B=ScalarParameter() {defaultValue=0.8, group=Colors, parameterName=Background Opacity}))
  n60 = Fresnel(ExponentIn=ScalarParameter() {defaultValue=4, group=Colors, parameterName=Opacity Fresnel Exponent}) {baseReflectFraction=0}
  n61 = Cosine(None=Multiply(A=Time()) {constB=0.3})
  n62 = ScalarParameter() {defaultValue=1, desc=from 0 to 1 for spawning, group=BP Logic, parameterName=Spawn Animation Time, sliderMax=1, sortPriority=10}
  n63 = StaticSwitchParameter(True=Saturate(None=Add(A=@n61) {constB=0.5}), False=@n62) {desc=Uses TIME for spawn animation, group=0_DEBUG, parameterName=DEBUG MODE}
  Multiply(A=Multiply(A=Saturate(None=@n59), B=OneMinus(None=@n60)), B=@n63)
  n64 = VectorParameter() {group=Animation, parameterName="Scale over time (xX yY)", sortPriority=10, defaultValue=<rgba 0 50 0 50>}
  n65 = OneMinus(None=@n63)
  n66 = Multiply(A=LinearInterpolate(A=@n64.B, B=@n64.B, Alpha=@n65), B=Transform(None=Constant3Vector() {constant=<rgba 0 -1 0 0>}))
  n67 = Subtract(A=TextureCoordinate()) {constB=0.5}
  n68 = Sign(None=ComponentMask(None=@n67) {g=true})
  n69 = Multiply(A=LinearInterpolate(A=@n64.R, B=@n64.R, Alpha=@n65), B=Transform(None=Constant3Vector() {constant=<rgba 1 0 0 0>}))
  n70 = Sign(None=ComponentMask(None=@n67) {r=true})
  Reroute(None=Add(A=Multiply(A=@n66, B=@n68), B=Multiply(A=@n69, B=@n70)))
 ]
}
